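# Parent Default Lit material for keycard gadgets, exposing Diffuse, Normal and Masks texture parameters.
# An arctangent of the flattened camera-to-actor direction yields a yaw that rotates vertices about Z at the object pivot, so the card tracks the camera.
# That offset is summed with a BlackHoleFunction distortion driven by a world-position parameter.
# The normal map is transformed and rotated by the same yaw, with tangent-space normals disabled to match.
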Material /Game/Gadgets/Assets/KeycardAlter/Materials/M_Keycard_Parent
graph {
 properties {bTangentSpaceNormal=false, pixelDepthOffsetMode=PDOM_Legacy, usedShadingModels=Default Lit}
 nodes [
  n1 = VectorParameter() {parameterName=WorldPostionOfBlackHole, defaultValue=<rgba -30000 25536 1952 0>}
  n2 = Constant3Vector() {constant=<rgba 0 0 1 0>}
  n3 = Normalize(VectorInput=Constant3Vector() {constant=<rgba 0 1 0 1>})
  n4 = ActorPositionWS()
  n5 = CameraPositionWS()
  n6 = Multiply(A=Normalize(VectorInput=Subtract(A=@n4, B=@n5)), B=Constant3Vector() {constant=<rgba 1 1 0 0>})
  n7 = Normalize(VectorInput=Constant3Vector() {constant=<rgba 1 0 0 1>})
  n8 = Add(
   A=Arctangent2Fast(Y=DotProduct(A=@n3, B=@n6), X=DotProduct(A=@n6, B=@n7))
   B=MaterialFunctionCall() {materialFunction=&Pi}.Result
  )
  Add(
   A=MaterialFunctionCall!(WorldPostionOfBlackHole=AppendVector(A=@n1.RGB, B=@n1.RGB), World Position) {materialFunction=/Game/Packages/Fortnite_UniqueMaterials_1/BlackHoleFunction}.Result
   B=RotateAboutAxis(
    NormalizedRotationAxis=@n2
    RotationAngle=@n8
    PivotPoint=MaterialFunctionCall() {materialFunction=&ObjectPivotPoint}.Object Pivot Location
    Position=WorldPosition().XYZ
   ) {period=6.2832}
  )
  n9 = TextureSampleParameter2D() {parameterName=Normal, samplerType=SAMPLERTYPE_Normal, texture=/Game/Gadgets/Assets/KeycardAlter/Textures/Alter/T_Item_Keycard_Alter_N}
  n10 = Transform(None=@n9.RGB)
  Add(A=RotateAboutAxis(NormalizedRotationAxis=@n2, RotationAngle=@n8, PivotPoint=Constant(), Position=@n10) {period=6.2832}, B=@n10)
  TextureSampleParameter2D() {parameterName=Diffuse, texture=/Game/Gadgets/Assets/KeycardAlter/Textures/MountainBase/T_Item_Keycard_Mountain_Base_D}
  TextureSampleParameter2D() {
   parameterName=Masks
   samplerType=SAMPLERTYPE_Masks
   texture=/Game/Gadgets/Assets/KeycardAlter/Textures/MountainBase/T_Item_Keycard_Mountain_Base_S
  }
 ]
}
